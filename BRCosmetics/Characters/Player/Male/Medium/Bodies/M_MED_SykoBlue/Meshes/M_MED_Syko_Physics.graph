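# Physics asset for the Syko player cosmetic body mesh (male, medium, SykoBlue), covering only its dangling torso tubes.
# Each of the four tubes, A and B on left and right, pairs a small capsule on its base bone with a larger offset capsule on the tube bone.
# A constraint joins each tube to its own base with equal, moderate swing and twist angles, with no links between tubes.
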PhysicsAsset /BRCosmetics/Characters/Player/Male/Medium/Bodies/M_MED_SykoBlue/Meshes/M_MED_Syko_Physics
bodies {
 dyn_torso_tube_a_base_l [{shapeType=Capsule, length=1, radius=1}]
 dyn_torso_tube_a_base_r [{shapeType=Capsule, length=1, radius=1}]
 dyn_torso_tube_a_l [{shapeType=Capsule, length=5.5, radius=2.25, center=<0 -0.0001 -4.8281>}]
 dyn_torso_tube_a_r [{shapeType=Capsule, length=5.5, radius=2.25, center=<0 0.0001 4.8281>, rotation=<rot 0 0 180>}]
 dyn_torso_tube_b_base_l [{shapeType=Capsule, length=1, radius=1}]
 dyn_torso_tube_b_base_r [{shapeType=Capsule, length=1, radius=1}]
 dyn_torso_tube_b_l [{shapeType=Capsule, length=5, radius=2.75, center=<0 0 -3.6554>}]
 dyn_torso_tube_b_r [{shapeType=Capsule, length=5, radius=2.75, center=<0 0 3.6554>, rotation=<rot 0 0 180>}]
}
constraints [
 {bone1Name=dyn_torso_tube_b_l, bone2Name=dyn_torso_tube_b_base_l, swing1LimitDegrees=45, swing2LimitDegrees=45, twistLimitDegrees=45}
 {bone1Name=dyn_torso_tube_a_l, bone2Name=dyn_torso_tube_a_base_l, swing1LimitDegrees=45, swing2LimitDegrees=45, twistLimitDegrees=45}
 {bone1Name=dyn_torso_tube_a_r, bone2Name=dyn_torso_tube_a_base_r, swing1LimitDegrees=45, swing2LimitDegrees=45, twistLimitDegrees=45}
 {bone1Name=dyn_torso_tube_b_r, bone2Name=dyn_torso_tube_b_base_r, swing1LimitDegrees=45, swing2LimitDegrees=45, twistLimitDegrees=45}
]
