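# Animation Blueprint for the M_MED_Syko (SykoBlue) male body cosmetic, copying the attached parent's pose and adding secondary motion to costume bones.
# Weighted constraints drive torso tubes, backpack socket and hip bases from spine, pelvis and thigh twist bones.
# Each update measures upper arm pitch and thigh yaw against ref pose, fading in corrective shoulder, arm twist and hip offsets.
# A pelvis-space RigidBody node then simulates the Syko physics asset.
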AnimBlueprint /BRCosmetics/Characters/Player/Male/Medium/Bodies/M_MED_SykoBlue/Meshes/M_MED_Syko_AnimBP
target_skeleton=&/M_MED_Syko_Skeleton
graphs [
 AnimGraph {nodes [
   n1 = "Constraint - dyn_torso_tube_a_base_l (PPPP)"!(
    ComponentPose=Local To Component(LocalPose=Copy Pose From Mesh() {bUseAttachedParent=true})
    Alpha=1
    ConstraintWeights_0=0.0181
    ConstraintWeights_1:float=0.0307
    ConstraintWeights_2:float=0.6139
    ConstraintWeights_3:float=0.3373
   ) -> (Pose) {
    alphaBoolBlend {blendOption=Linear}
    boneToModify {boneName=dyn_torso_tube_a_base_l}
    constraintSetup [{targetBone {boneName=spine_01}}, {targetBone {boneName=spine_02}}, {targetBone {boneName=spine_03}}, {targetBone {boneName=spine_04}}]
    constraintWeights [1,1,1,1]
   }
   n2 = "Constraint - dyn_torso_tube_b_base_l (PPPPP)"!(
    ComponentPose=@n1
    Alpha=1
    ConstraintWeights_0=0.0921
    ConstraintWeights_1:float=0.1167
    ConstraintWeights_2:float=0.588
    ConstraintWeights_3:float=0.1847
    ConstraintWeights_4:float=0.0185
   ) -> (Pose) {
    alphaBoolBlend {blendOption=Linear}
    boneToModify {boneName=dyn_torso_tube_b_base_l}
    constraintSetup [
     {targetBone {boneName=spine_01}}
     {targetBone {boneName=spine_02}}
     {targetBone {boneName=spine_03}}
     {targetBone {boneName=spine_04}}
     {targetBone {boneName=pelvis}}
    ]
    constraintWeights [1,1,1,1,1]
   }
   n3 = "Constraint - dyn_torso_tube_a_base_r (PPPP)"!(
    ComponentPose=@n2
    Alpha=1
    ConstraintWeights_0=0.0181
    ConstraintWeights_1:float=0.0307
    ConstraintWeights_2:float=0.6139
    ConstraintWeights_3:float=0.3373
   ) -> (Pose) {
    alphaBoolBlend {blendOption=Linear}
    boneToModify {boneName=dyn_torso_tube_a_base_r}
    constraintSetup [{targetBone {boneName=spine_01}}, {targetBone {boneName=spine_02}}, {targetBone {boneName=spine_03}}, {targetBone {boneName=spine_04}}]
    constraintWeights [1,1,1,1]
   }
   n4 = "Constraint - dyn_torso_tube_b_base_r (PPPPP)"!(
    ComponentPose=@n3
    Alpha=1
    ConstraintWeights_0=0.0921
    ConstraintWeights_1:float=0.1167
    ConstraintWeights_2:float=0.588
    ConstraintWeights_3:float=0.1847
    ConstraintWeights_4:float=0.0185
   ) -> (Pose) {
    alphaBoolBlend {blendOption=Linear}
    boneToModify {boneName=dyn_torso_tube_b_base_r}
    constraintSetup [
     {targetBone {boneName=spine_01}}
     {targetBone {boneName=spine_02}}
     {targetBone {boneName=spine_03}}
     {targetBone {boneName=spine_04}}
     {targetBone {boneName=pelvis}}
    ]
    constraintWeights [1,1,1,1,1]
   }
   n5 = "Transform (Modify) Bone - Bone: dyn_shldr_lock_l"(ComponentPose=@n4, Alpha=Get armUp_L() -> (armUp_L:double), Translation="2.5,0,0", Rotation="0,0,0", Scale="1,1,1") {
    rotationMode=BMM_Additive
    rotationSpace=BCS_BoneSpace
    scaleMode=BMM_Additive
    scaleSpace=BCS_BoneSpace
    translationMode=BMM_Additive
    translationSpace=BCS_BoneSpace
    alphaBoolBlend {blendOption=Linear}
    alphaScaleBiasClamp {bMapRange=true, inRange {max=40}}
    boneToModify {boneName=dyn_shldr_lock_l}
    translation {x=2.5}
   }
   n6 = "Transform (Modify) Bone - Bone: dyn_shldr_rock_r"(ComponentPose=@n5, Alpha=Get armUp_R() -> (armUp_R:double), Translation="-2.5,0,0", Rotation="0,0,0", Scale="1,1,1") {
    rotationMode=BMM_Additive
    rotationSpace=BCS_BoneSpace
    scaleMode=BMM_Additive
    scaleSpace=BCS_BoneSpace
    translationMode=BMM_Additive
    translationSpace=BCS_BoneSpace
    alphaBoolBlend {blendOption=Linear}
    alphaScaleBiasClamp {bMapRange=true, inRange {max=15}}
    boneToModify {boneName=dyn_shldr_rock_r}
    translation {x=-2.5}
   }
   n7 = "Transform (Modify) Bone - Bone: upperarm_twist_01_l"(ComponentPose=@n6, Alpha=Get armUp_L(), Translation="0,-2,0", Rotation="0,0,0", Scale="1,1,1") {
    rotationMode=BMM_Additive
    rotationSpace=BCS_BoneSpace
    scaleMode=BMM_Additive
    scaleSpace=BCS_BoneSpace
    translationMode=BMM_Additive
    translationSpace=BCS_BoneSpace
    alphaBoolBlend {blendOption=Linear}
    alphaScaleBiasClamp {bMapRange=true, inRange {max=40}}
    boneToModify {boneName=upperarm_twist_01_l}
    translation {y=-2}
   }
   n8 = "Transform (Modify) Bone - Bone: upperarm_twist_01_r"(ComponentPose=@n7, Alpha=Get armUp_R(), Translation="0,2,4", Rotation="0,0,0", Scale="1,1,1") {
    rotationMode=BMM_Additive
    rotationSpace=BCS_BoneSpace
    scaleMode=BMM_Additive
    scaleSpace=BCS_BoneSpace
    translationMode=BMM_Additive
    translationSpace=BCS_BoneSpace
    alphaBoolBlend {blendOption=Linear}
    alphaScaleBiasClamp {bMapRange=true, inRange {max=15}}
    boneToModify {boneName=upperarm_twist_01_r}
    translation {y=2, z=4}
   }
   n9 = "Constraint - backpack_socket (PPP)"!(ComponentPose=@n8, Alpha=1, ConstraintWeights_0=0.627, ConstraintWeights_1:float=0.2418, ConstraintWeights_2:float=0.1313) -> (Pose) {
    alphaBoolBlend {blendOption=Linear}
    boneToModify {boneName=backpack_socket}
    constraintSetup [{targetBone {boneName=spine_03}}, {targetBone {boneName=spine_04}}, {targetBone {boneName=spine_05}}]
    constraintWeights [1,1,1]
   }
   n10 = "Constraint - dyn_hip_base_l (R)"(ComponentPose=@n9, Alpha=0.7, ConstraintWeights_0=1) {
    alphaBoolBlend {blendOption=Linear}
    boneToModify {boneName=dyn_hip_base_l}
    constraintSetup [{transformType=Rotation, targetBone {boneName=thigh_twist_01_l}}]
    constraintWeights [1]
   }
   n11 = "Transform (Modify) Bone - Bone: dyn_hip_l"(ComponentPose=@n10, Alpha=Get legUp_L() -> (legUp_L:double), Translation="2,3,1", Rotation="0,0,0", Scale="1,1,1") {
    rotationMode=BMM_Additive
    rotationSpace=BCS_BoneSpace
    scaleMode=BMM_Additive
    scaleSpace=BCS_BoneSpace
    translationMode=BMM_Additive
    translationSpace=BCS_BoneSpace
    alphaBoolBlend {blendOption=Linear}
    alphaScaleBiasClamp {bMapRange=true, inRange {max=0, min=-110}, outRange {max=0, min=1}}
    boneToModify {boneName=dyn_hip_l}
    translation=<2 3 1>
   }
   n12 = "Constraint - dyn_hip_base_r (R)"(ComponentPose=@n11, Alpha=0.7, ConstraintWeights_0=1) {
    alphaBoolBlend {blendOption=Linear}
    boneToModify {boneName=dyn_hip_base_r}
    constraintSetup [{transformType=Rotation, targetBone {boneName=thigh_twist_01_r}}]
    constraintWeights [1]
   }
   n13 = "Transform (Modify) Bone - Bone: dyn_hip_r"(ComponentPose=@n12, Alpha=Get legUp_R() -> (legUp_R:double), Translation="-2,-3,-1", Rotation="0,0,0", Scale="1,1,1") {
    rotationMode=BMM_Additive
    rotationSpace=BCS_BoneSpace
    scaleMode=BMM_Additive
    scaleSpace=BCS_BoneSpace
    translationMode=BMM_Additive
    translationSpace=BCS_BoneSpace
    alphaBoolBlend {blendOption=Linear}
    alphaScaleBiasClamp {bMapRange=true, inRange {max=0, min=-110}, outRange {max=0, min=1}}
    boneToModify {boneName=dyn_hip_r}
    translation=<-2 -3 -1>
   }
   n14 = RigidBody(ComponentPose=@n13) {
    bClampLinearTranslationLimitToRefPose=true
    overridePhysicsAsset=&/M_MED_Syko_Physics
    simulationSpace=BaseBoneSpace
    alphaBoolBlend {blendOption=Linear}
    baseBoneRef {boneName=pelvis}
   }
   Output Pose(Result=Component To Local(ComponentPose=@n14)) {layerGroup=None}
  ]}
 EventGraph {nodes [
   Event BlueprintUpdateAnimation()
   > n1 = Sequence!(execute) -> (then_0, then_1, then_2:Exec, then_3:Exec)
   n2 = GetOwnerSkeletalMeshComponent()
   n3 = GetDeltaTransformFromRefPose(self=@n2, BoneName=upperarm_l, BaseName=clavicle_l)
   > Set armUp_L(armUp_L:double=BreakRotator(InRot=BreakTransform(InTransform=@n3).Rotation).Pitch) -> (Output_Get:double)
   n4 = GetDeltaTransformFromRefPose(self=@n2, BoneName=upperarm_r, BaseName=clavicle_r)
   Set armUp_R(execute=@n1.then_1, armUp_R:double=BreakRotator(InRot=BreakTransform(InTransform=@n4).Rotation).Pitch) -> (Output_Get:double)
   n5 = GetDeltaTransformFromRefPose(self=@n2, BoneName=thigh_l, BaseName=pelvis)
   Set legUp_L(execute=@n1.then_2, legUp_L:double=BreakRotator(InRot=BreakTransform(InTransform=@n5).Rotation).Yaw) -> (Output_Get:double)
   n6 = GetDeltaTransformFromRefPose(self=@n2, BoneName=thigh_r, BaseName=pelvis)
   Set legUp_R(execute=@n1.then_3, legUp_R:double=BreakRotator(InRot=BreakTransform(InTransform=@n6).Rotation).Yaw) -> (Output_Get:double)
   TryGetPawnOwner()
  ]}
]
variables [
 armUp_L {type=number}
 armUp_R {type=number}
 legUp_L {type=number}
 legUp_R {type=number}
]
